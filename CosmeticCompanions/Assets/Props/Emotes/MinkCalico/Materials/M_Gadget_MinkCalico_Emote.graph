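# Cloth-shading material for the Mink Calico emote gadget prop, built on the 2023 anime character shader functions with optional eye-animation, nose-line and Kicks switches.
# It adds cycling lights: an FX texture offsets scaled time, and the wrapped result selects red, green or yellow.
# The desaturated, boosted shader emissive masks these lights, blending in black base color and colored emissive.
# Parameters control pulse speed, color variation, intensity and flash toggle.
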Material /CosmeticCompanions/Assets/Props/Emotes/MinkCalico/Materials/M_Gadget_MinkCalico_Emote
graph {
 expressions {
  MaterialExpressionAdd_0 {
   class=MaterialExpressionAdd
   inputs [
    A {from=MaterialExpressionMultiply_9}
    B {from=MaterialExpressionMultiply_10}
   ]
  }
  MaterialExpressionAppendVector_0 {
   class=MaterialExpressionAppendVector
   inputs [
    A {from=MaterialExpressionVertexColor_0, output=""}
    B {from=MaterialExpressionVertexColor_0, output=""}
   ]
  }
  MaterialExpressionBlendMaterialAttributes_0 {
   class=MaterialExpressionBlendMaterialAttributes
   inputs [
    A {from=MaterialExpressionGetMaterialAttributes_1, output=MaterialAttributes}
    B {from=MaterialExpressionSetMaterialAttributes_1, output=""}
    Alpha {from=MaterialExpressionReroute_2, output=""}
   ]
   properties {}
  }
  MaterialExpressionChannelMaskParameter_0 {
   class=MaterialExpressionChannelMaskParameter
   inputs [None {from=MaterialExpressionStaticSwitchParameter_1, output=""}]
   properties {group=01_EyeAnimation_Parameters, parameterName=EyeMaskChannel, sortPriority=-32}
  }
  MaterialExpressionComponentMask_0 {class=MaterialExpressionComponentMask, inputs [None {from=MaterialExpressionTextureCoordinate_0}], properties {r=true}}
  MaterialExpressionConstant_0 {class=MaterialExpressionConstant}
  MaterialExpressionDesaturation_0 {
   class=MaterialExpressionDesaturation
   inputs [
    None {from=MaterialExpressionMaterialFunctionCall_21, output=Emissive}
    Fraction {from="", output=""}
   ]
   properties {}
  }
  MaterialExpressionFmod_1 {
   class=MaterialExpressionFmod
   inputs [
    A {from=MaterialExpressionAdd_0, output=""}
    B {from=MaterialExpressionReroute_0, output=""}
   ]
   properties {}
  }
  MaterialExpressionGetMaterialAttributes_1 {
   class=MaterialExpressionGetMaterialAttributes
   inputs [None {from=MaterialExpressionMaterialFunctionCall_14, output=Result}]
   properties {attributeGetTypes [#1]}
  }
  MaterialExpressionMaterialFunctionCall_0 {
   class=MaterialExpressionMaterialFunctionCall
   inputs [
    Color Lit {from=MaterialExpressionMaterialFunctionCall_21, output=Color_Lit}
    Color Shaded {from=MaterialExpressionMaterialFunctionCall_21, output=Color_Lit}
    Use UV2 {from=MaterialExpressionStaticBoolParameter_0}
    Mask {from=MaterialExpressionStaticSwitchParameter_0}
   ]
   properties {materialFunction=/Game/Global/Materials/Character_Parent_2023/Anime_Parent_2023/MaterialFunctions/MF_AnimeEyeAnimation_v2}
  }
  MaterialExpressionMaterialFunctionCall_1 {
   class=MaterialExpressionMaterialFunctionCall
   inputs []
   properties {materialFunction=/Game/Global/Materials/Character_Parent_2023/Anime_Parent_2023/MaterialFunctions/MF_NoseLineFlip}
  }
  MaterialExpressionMaterialFunctionCall_14 {
   class=MaterialExpressionMaterialFunctionCall
   inputs [
    Shading_Gradient {from=MaterialExpressionMaterialFunctionCall_20, output=Shading_Gradient}
    Threshold_Gradient {from=MaterialExpressionMaterialFunctionCall_20, output=Shading_Gradient}
    Specular_Gradient {from=MaterialExpressionMaterialFunctionCall_20, output=Shading_Gradient}
    Sheen_Mask {from=MaterialExpressionMaterialFunctionCall_20, output=Shading_Gradient}
    Thickness {from=MaterialExpressionMaterialFunctionCall_20, output=Shading_Gradient}
    DFL_Mask {from=MaterialExpressionMaterialFunctionCall_20, output=Shading_Gradient}
    FX {from=MaterialExpressionMaterialFunctionCall_20, output=Shading_Gradient}
    Metallic_Gradient {from=MaterialExpressionMaterialFunctionCall_20, output=Shading_Gradient}
    RimLight_Gradient {from=MaterialExpressionMaterialFunctionCall_20, output=Shading_Gradient}
    NormalBuffer {from=MaterialExpressionMaterialFunctionCall_20, output=Shading_Gradient}
    -------- {from=""}
    CINE_LightSources {from=MaterialExpressionMaterialFunctionCall_20, output=Shading_Gradient}
    CINE_LightColor {from=MaterialExpressionMaterialFunctionCall_20, output=Shading_Gradient}
    --------- {from=""}
    Color_Lit {from=MaterialExpressionMaterialFunctionCall_0, output=Lit}
    Color_Shaded {from=MaterialExpressionMaterialFunctionCall_0, output=Lit}
    SRM {from=MaterialExpressionMaterialFunctionCall_21, output=SRM}
    Emissive {from=MaterialExpressionMaterialFunctionCall_21, output=SRM}
    DFL_Color {from=MaterialExpressionMaterialFunctionCall_21, output=SRM}
    Specular_Color {from=MaterialExpressionMaterialFunctionCall_21, output=SRM}
    RimLight_Color {from=MaterialExpressionMaterialFunctionCall_21, output=SRM}
   ]
   properties {materialFunction=/Game/Global/Materials/Character_Parent_2023/Anime_Parent_2023/MaterialFunctions/MFC_AnimeShader_2023_Post}
  }
  MaterialExpressionMaterialFunctionCall_2 {
   class=MaterialExpressionMaterialFunctionCall
   inputs [
    InOpacity {from=""}
    Material {from=MaterialExpressionBlendMaterialAttributes_0}
    Mask {from=""}
   ]
   properties {materialFunction=/Game/Global/Materials/Cosmetics/Kicks/Socks/Materials/MF_Sock}
  }
  MaterialExpressionMaterialFunctionCall_20 {
   class=MaterialExpressionMaterialFunctionCall
   inputs [
    CustomLightVector {from=""}
    OverrideNormal_WS {from=""}
    STT {from=MaterialExpressionMaterialFunctionCall_21, output=STT}
    Normal_Map {from=MaterialExpressionMaterialFunctionCall_21, output=STT}
    DFL {from=MaterialExpressionStaticSwitchParameter_10}
    FX {from=MaterialExpressionMaterialFunctionCall_21, output=STT}
   ]
   properties {materialFunction=/Game/Global/Materials/Character_Parent_2023/Anime_Parent_2023/MaterialFunctions/MFC_AnimeShader_2023_Calculations}
  }
  MaterialExpressionMaterialFunctionCall_21 {
   class=MaterialExpressionMaterialFunctionCall
   inputs [CustomUV {from=""}]
   properties {materialFunction=/Game/Global/Materials/Character_Parent_2023/Anime_Parent_2023/MaterialFunctions/MFC_AnimeShader_2023_Pre}
  }
  MaterialExpressionMultiply_10 {
   class=MaterialExpressionMultiply
   inputs [
    A {from=MaterialExpressionTime_0}
    B {from=MaterialExpressionScalarParameter_1}
   ]
  }
  MaterialExpressionMultiply_2 {
   class=MaterialExpressionMultiply
   inputs [
    A {from=MaterialExpressionDesaturation_0}
    B {from=""}
   ]
   properties {constB=3}
  }
  MaterialExpressionMultiply_4 {
   class=MaterialExpressionMultiply
   inputs [
    A {from=MaterialExpressionSaturate_2}
    B {from=MaterialExpressionMultiply_7}
   ]
  }
  MaterialExpressionMultiply_5 {
   class=MaterialExpressionMultiply
   inputs [
    A {from=MaterialExpressionMultiply_4}
    B {from=MaterialExpressionScalarParameter_8}
   ]
  }
  MaterialExpressionMultiply_7 {
   class=MaterialExpressionMultiply
   inputs [
    A {from=MaterialExpressionSwitch_0}
    B {from=MaterialExpressionScalarParameter_10}
   ]
  }
  MaterialExpressionMultiply_9 {
   class=MaterialExpressionMultiply
   inputs [
    A {from=MaterialExpressionTextureSample_2, output=R}
    B {from=MaterialExpressionScalarParameter_2}
   ]
   properties {constB=3}
  }
  MaterialExpressionReroute_0 {class=MaterialExpressionReroute, inputs [None {from=MaterialExpressionReroute_1}]}
  MaterialExpressionReroute_1 {class=MaterialExpressionReroute, inputs [None {from=MaterialExpressionScalarParameter_2}]}
  MaterialExpressionReroute_2 {class=MaterialExpressionReroute, inputs [None {from=MaterialExpressionSaturate_2}]}
  MaterialExpressionSaturate_0 {class=MaterialExpressionSaturate, inputs [None {from=MaterialExpressionSign_0}]}
  MaterialExpressionSaturate_2 {class=MaterialExpressionSaturate, inputs [None {from=MaterialExpressionMultiply_2}]}
  MaterialExpressionScalarParameter_1 {class=MaterialExpressionScalarParameter, properties {defaultValue=3, group=Mink Calico - Emote, parameterName=Light Pulse - Speed}}
  MaterialExpressionScalarParameter_10 {class=MaterialExpressionScalarParameter, properties {defaultValue=1, group=Mink Calico - Emote, parameterName=Light Intensity}}
  MaterialExpressionScalarParameter_2 {class=MaterialExpressionScalarParameter, properties {defaultValue=3, group=Mink Calico - Emote, parameterName=Color Variation}}
  MaterialExpressionScalarParameter_8 {class=MaterialExpressionScalarParameter, properties {defaultValue=1, group=Mink Calico - Emote, parameterName=Toggle Flash}}
  MaterialExpressionSetMaterialAttributes_1 {
   class=MaterialExpressionSetMaterialAttributes
   inputs [
    MaterialAttributes {from=MaterialExpressionGetMaterialAttributes_1, output=MaterialAttributes}
    Emissive Color {from=MaterialExpressionMultiply_5, output=""}
    Base Color {from=MaterialExpressionConstant_0, output=""}
   ]
   properties {attributeSetTypes [#1, #2]}
  }
  MaterialExpressionSign_0 {class=MaterialExpressionSign, inputs [None {from=MaterialExpressionSubtract_0, output=""}], properties {}}
  MaterialExpressionStaticBoolParameter_0 {
   class=MaterialExpressionStaticBoolParameter
   inputs []
   properties {defaultValue=true, group=01_EyeAnimation_Parameters, parameterName=Use UV2 for Eye Animation, sortPriority=-30}
  }
  MaterialExpressionStaticSwitchParameter_0 {
   class=MaterialExpressionStaticSwitchParameter
   inputs [
    True {from=MaterialExpressionChannelMaskParameter_0, output=R}
    False {from=MaterialExpressionSaturate_0, output=""}
   ]
   properties {group=01_EyeAnimation_Parameters, parameterName=UseEyeMask, sortPriority=-33}
  }
  MaterialExpressionStaticSwitchParameter_1 {
   class=MaterialExpressionStaticSwitchParameter
   inputs [
    True {from=MaterialExpressionAppendVector_0, output=""}
    False {from=MaterialExpressionTextureSampleParameter2D_0, output=RGBA}
   ]
   properties {group=01_EyeAnimation_Parameters, parameterName=UseVertexColorForEyeMask, sortPriority=-33}
  }
  MaterialExpressionStaticSwitchParameter_10 {
   class=MaterialExpressionStaticSwitchParameter
   inputs [
    True {from=MaterialExpressionMaterialFunctionCall_1, output=DFL}
    False {from=MaterialExpressionMaterialFunctionCall_21, output=DFL}
   ]
   properties {group=Nose, parameterName=UseNoseLineFlip, sortPriority=0}
  }
  MaterialExpressionStaticSwitchParameter_2 {
   class=MaterialExpressionStaticSwitchParameter
   inputs [
    True {from=MaterialExpressionMaterialFunctionCall_2, output=Result}
    False {from=MaterialExpressionBlendMaterialAttributes_0, output=""}
   ]
   properties {group=Kicks, parameterName=IsKicks}
  }
  MaterialExpressionSubtract_0 {
   class=MaterialExpressionSubtract
   inputs [
    A {from=MaterialExpressionComponentMask_0}
    B {from=""}
   ]
   properties {constB=0.0001}
  }
  MaterialExpressionSwitch_0 {
   class=MaterialExpressionSwitch
   inputs [
    SwitchValue {from=MaterialExpressionFmod_1, output=""}
    Default {from="", output=""}
    None {from=MaterialExpressionVectorParameter_6, output=RGB}
    None {from=MaterialExpressionVectorParameter_7, output=RGB}
    None {from=MaterialExpressionVectorParameter_8, output=RGB}
   ]
   properties {inputs [
     {
      inputName=None
      input {
       expression=&self:MaterialExpressionVectorParameter_6
       inputName=None
       mask=1
       maskA=0
       maskB=1
       maskG=1
       maskR=1
       outputIndex=0
      }
     }
     {
      inputName=None
      input {
       expression=&self:MaterialExpressionVectorParameter_7
       inputName=None
       mask=1
       maskA=0
       maskB=1
       maskG=1
       maskR=1
       outputIndex=0
      }
     }
     {
      inputName=None
      input {
       expression=&self:MaterialExpressionVectorParameter_8
       inputName=None
       mask=1
       maskA=0
       maskB=1
       maskG=1
       maskR=1
       outputIndex=0
      }
     }
    ]}
  }
  MaterialExpressionTemporalResponsivenessOutput_0 {class=MaterialExpressionTemporalResponsivenessOutput, inputs [None {from="", output=""}], properties {}}
  MaterialExpressionTextureCoordinate_0 {class=MaterialExpressionTextureCoordinate, properties {coordinateIndex=1}}
  MaterialExpressionTextureSampleParameter2D_0 {
   class=MaterialExpressionTextureSampleParameter2D
   inputs [
    UVs {from="", output=""}
    Apply View MipBias {from="", output=""}
   ]
   properties {
    group=01_EyeAnimation_Parameters
    parameterName=EyeMask
    samplerSource=SSM_Wrap_WorldGroupSettings
    samplerType=SAMPLERTYPE_Masks
    sortPriority=-32
    texture=&T_WhiteMask
   }
  }
  MaterialExpressionTextureSample_2 {
   class=MaterialExpressionTextureSample
   inputs [
    UVs {from=""}
    Tex {from=""}
    Apply View MipBias {from=""}
   ]
   properties {
    samplerSource=SSM_Wrap_WorldGroupSettings
    samplerType=SAMPLERTYPE_LinearGrayscale
    texture=/CosmeticCompanions/Assets/Props/Emotes/MinkCalico/Textures/T_MinkCalico_EmoteProp_FX1
   }
  }
  MaterialExpressionTime_0 {class=MaterialExpressionTime}
  MaterialExpressionVectorParameter_6 {class=MaterialExpressionVectorParameter, properties {group=Mink Calico - Emote, parameterName=Red, defaultValue=<rgba 3 0 0 0>}}
  MaterialExpressionVectorParameter_7 {class=MaterialExpressionVectorParameter, properties {group=Mink Calico - Emote, parameterName=Green, defaultValue=<rgba 0 2 0 1>}}
  MaterialExpressionVectorParameter_8 {class=MaterialExpressionVectorParameter, properties {group=Mink Calico - Emote, parameterName=Yellow, defaultValue=<rgba 1.5 1.5 0 1>}}
  MaterialExpressionVertexColor_0 {class=MaterialExpressionVertexColor, inputs [], properties {}}
 }
 properties {
  bAutomaticallySetUsageInEditor=false
  bUseMaterialAttributes=true
  bUsedWithClothing=true
  bUsedWithMeshDeformer=true
  bUsedWithMorphTargets=true
  bUsedWithNiagaraMeshParticles=true
  bUsedWithSkeletalMesh=true
  bUsedWithStaticLighting=true
  pixelDepthOffsetMode=PDOM_Legacy
  shadingModel=MSM_Cloth
  usedShadingModels=Cloth
 }
}
